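# Idle Niagara effect for the RankedTrooper pickaxe, layering a glow with electric arcs tinted by User.VariantColor.
# Glows drives a mesh and sprite with waveform-flickered color.
# Lightning spawns waveform-modulated bursts of short-lived meshes and glow sprites that slide along a fixed start-to-end line, brightening when bIsSwinging is set and scaling alpha by time of day.
# Lightning001, an array-indexed bolt variant with mirroring and inversion, is disabled.
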NiagaraSystem /BRCosmetics/Weapons/FORT_Melee/Pickaxe_RankedTrooper/FX/NS_Pickaxe_RankedTrooper_Idle
dependencies {
 usedDataInterfaces [
  &NiagaraDataInterfaceCurve
  &NiagaraDataInterfaceArrayFloat
  &NiagaraDataInterfaceMeshRendererInfo
  &NiagaraDataInterfaceColorCurve
 ]
 usedDynamicInputs [
  &SimulationPosition
  &SelectFloatFromArray
  &Multiply_Int
 ]
 usedModules [
  &self:SystemSpawnScript.SystemScriptSource.SystemScriptGraph.NiagaraNodeAssignment_0.SetVariables_#1
  &SystemState
  &self:SystemSpawnScript.SystemScriptSource.SystemScriptGraph.NiagaraNodeAssignment_1.SetVariables_#2
 ]
}
emitters {
 Glows {
  emitterUpdateScript [
   &EmitterState(Life Cycle Mode=::NewEnumerator1, Loop Behavior=::NewEnumerator1, Loop Duration Mode=::NewEnumerator1, UseLoopDelay=false)
   &SpawnBurst_Instantaneous(Use Spawn Probability=false, UseLoopCountLimit=true)
   /SetVariables(Emitter.CapsuleGlowParam:Vector4f=<3 0 0 0>) {isSetParameters=true}
   /Niagara/Noise/Waveform(
    EnableExponent1=false
    Global Input Position:NiagaraFloat=&RandomRangeFloat(Minimum=0, Maximum=0.1)
    GlobalClampMax=false
    GlobalClampMin=false
    UseInput1=false
    UsePhase1=false
    "[1] Frequency":NiagaraFloat=&RandomRangeFloat(Minimum=0.15, Maximum=0.25)
   ) {enabled=false, module=Waveform001}
  ]
  particleSpawnScript [
   &InitializeParticle(
    Color=<rgba 25 22.6667 22.5 0.5>
    Mesh Scale=&Multiply_VectorByFloat(Vector=<1.0573 1 1.02>, Float=&Multiply_Float(A=1, B=$User.FXScale))
    Mesh Scale Mode=::NewEnumerator4
    Position Offset=<0 0 15>
    Sprite Rotation Mode=::NewEnumerator0
    Sprite Size=&Multiply_Vector2D_ByFloat(Float=1, Vector2D=<50 110>)
    Sprite Size Mode=::NewEnumerator4
    UsePositionOffset=true
    Write Material Random=true
   )
   &SpriteFacingAndAlignment(Sprite Alignment=<0.05 0 1>, Sprite Facing=<0.7068 0.07 0.3>, WriteAlignment=true, WriteFacing=false)
  ]
  particleUpdateScript [
   &ParticleState(Kill Particles When Lifetime Has Elapsed=false, Let Infinitely Lived Particles Die When Emitter Deactivates=true)
   &DynamicMaterialParameters(
    Index 0 Param 1=4.0854
    Index 0 Param 2=1
    Index 0 Param 4=1
    Index 1 Param 1=12702.6572
    Index 1 Param 2=115.411
    Index 1 Param 3=50.9432
    Index 1 Param 4=24.9799
    Index 2 Param 1=-2
    Index 2 Param 2=2
    Index 2 Param 3=1
    Index 3 Param 3=1
    Index 3 Param 4=1
    Module.Index0Param0Write=true
    Module.Index0Param1Write=true
    Module.Index0Param2Write=false
    Module.Index0Param3Write=false
    Module.Index1Param0Write=false
    Module.Index1Param1Write=false
    Module.Index1Param2Write=false
    Module.Index1Param3Write=false
    Module.Index2Param0Write=true
    Module.Index2Param1Write=true
    Module.Index2Param2Write=true
    Module.Index2Param3Write=false
    Module.Index3Param0Write=true
    Module.Index3Param1Write=false
    Module.Index3Param2Write=true
    Module.Index3Param3Write=true
    Write Parameter Index 1=true
    Write Parameter Index 2=true
    Write Parameter Index 3=true
   )
   /SetVariables(Particles.ColorGlow:LinearColor=&Multiply_LinearColorByFloat(
     Linear Color=&MultiplyLinearColor(Linear Color=$Particles.Color, Scale Factor=&MakeVector4_FromLinearColor(LinearColor=$User.VariantColor))
     Float=/Niagara/DynamicInputs/Helpers/Waveform("Waveform [1]"=::NewEnumerator12, "[1] Frequency"=8, "[1] Amplitude Min/Max"=<0.75 1.1>, "[1] Scale"=0.89)
    )) {isSetParameters=true}
   &FN_Sprite_Attributes(
    Camera Offset=-10
    Color=&MultiplyLinearColor(
     Linear Color=$Particles.ColorGlow
     Scale Factor=&MakeFloat4FromVectorAndFloat("Vector (RGB)"=&VectorFromFloat(Value=0.05), "Float (Alpha)"=0.2)
    )
    Create Camera Offset Attribute=true
    Create Color Attribute=true
    Create Position Attribute=true
    Position:NiagaraPosition=&AddVectorToPosition(Position=/Niagara/DynamicInputs/Helpers/SimulationPosition(), Vector=<-2.4514 1 50.5126>)
    Write to Module Namespace=true
   )
   /SetVariables(Particles.CapsuleParams:Vector4f=&MakeVector4(X=3, Y=0, Z=0, W=0)) {isSetParameters=true}
  ]
  rendererClasses [&NiagaraMeshRendererProperties, &NiagaraSpriteRendererProperties]
  renderers [
   {rendererClass=&NiagaraMeshRendererProperties, rendererIndex=0}
   {rendererClass=&NiagaraSpriteRendererProperties, rendererIndex=1}
  ]
 }
 Lightning {
  emitterSpawnScript [/SetVariables(Emitter.EndPos:NiagaraPosition=<-1 1 75>, Emitter.StartPos:NiagaraPosition=<-3.5 1 20>) {isSetParameters=true}]
  emitterUpdateScript [
   &EmitterState(
    Inactive Response=::NewEnumerator1
    Life Cycle Mode=::NewEnumerator1
    Loop Duration=&RandomRangeFloat(Minimum=2, Maximum=7)
    Recalculate Duration Each Loop=true
    Scalability Mode=::NewEnumerator1
    UseLoopDelay=true
   )
   &SpawnRate(
    SpawnRate=&Multiply_Float(
     A=&Multiply_Float(
      A=&RandomRangeFloat(Minimum=5, Maximum=7.5)
      B=&Waveform(
       Waveform Count=::NewEnumerator1
       "Waveform [1]"=::NewEnumerator0
       "[1] Frequency"=1
       "[1] Amplitude Min/Max"=<0 1>
       "[1] Scale"=1.5
       "Waveform [2] Blend"=::NewEnumerator2
       "Waveform [2]"=::NewEnumerator12
       "[2] Input Position"=0
       "[2] Phase"=$Emitter.Age
       "[2] Phase Scale"=1
       "[2] Compound Frequency"=2
       "[2] Frequency"=12
       "[2] Amplitude Min/Max"=<0.1 1>
       "[2] Scale"=2
       "[2] Bias"=0
       "[2] Exponent"=1
       "[2] Pulse Duration"=0.1
       "[2] Clip Triangle Min/Max"=<0 0>
       "[2] Random Seed"=0
      )
     )
     B=&FloatFromCurve(FloatCurve=DataInterface() {propertyValues {Curve {keys [<at 0 1.5>,<at 0.2 1.125>,<at 1 1>]}}}, CurveIndex=$Emitter.NormalizedLoopAge)
    )
    Use Spawn Probability=false
   )
  ]
  particleSpawnScript [
   &InitializeParticle(
    Color=$User.VariantColor
    Lifetime Max=0.3
    Lifetime Min=0.125
    Lifetime Mode=::NewEnumerator1
    Mesh Scale Max=<0.4 0.4 0.6>
    Mesh Scale Min=<0.25 0.25 0.3>
    Mesh Scale Mode=::NewEnumerator2
    Position Offset=&RandomRangeVector(Minimum=<0 0 0>, Maximum=<0 0 0>)
    UsePositionOffset=false
    Write Material Random=true
   )
   &InitialMeshOrientation(
    Mesh Orientation Mode=::NewEnumerator2
    Rotation=&RandomRangeVector(Minimum=<0 0 -1>, Maximum=<0 0 1>)
    Use Rotation Vector=true
   ) {enabled=false}
   &AddVelocity(
    Module.Use Falloff Distance=false
    Override Seed=false
    Velocity Mode=::NewEnumerator1
    Velocity Origin=$Particles.Position
    Velocity Speed=10
   ) {enabled=false, module=AddVelocity001}
   /SetVariables(Particles.Mesh index:NiagaraInt32=&RandomRangeInt(Minimum=0, Maximum=2)) {isSetParameters=true}
   /SetVariables(Particles.InitialPosition:NiagaraFloat=&RandomRangeFloat(Minimum=0.2, Maximum=0.8)) {isSetParameters=true}
   /SetVariables(Particles.Position=&Lerp_Vector(
     A=Linked() {linkedVariable {name=Emitter.StartPos, type=NiagaraPosition}}
     B=Linked() {linkedVariable {name=Emitter.EndPos, type=NiagaraPosition}}
     Alpha=$Particles.InitialPosition
    )) {isSetParameters=true}
   &CameraOffset(Camera Offset Amount=&RandomRangeFloat(Minimum=-20, Maximum=20))
   &InitialMeshOrientation(
    Mesh Axis To Orient=<0 0 1>
    Mesh Orientation Mode=::NewEnumerator2
    Orientation Vector=&NormalizeVector(
     Vector To Normalize=&Subtract_Position(A=$Emitter.StartPos, B=$Emitter.EndPos)
     Safe Normalize=true
     Safe Fallback Threshold=0
     Safe Fallback Vector=<0 0 0.001>
    )
    Rotation=&MakeVector(X=0, Y=0, Z=&RandomRangeFloat(Minimum=0, Maximum=1))
    Use Rotation Vector=true
   ) {module=InitialMeshOrientation001}
  ]
  particleUpdateScript [
   &SolveForcesAndVelocity(Clamp Velocity=false, Limit Acceleration=false)
   &SkeletalMeshLocation(
    MaskSpawnGroup=false
    Mesh Sampling Type=::NewEnumerator5
    OverrideRandomSeed=false
    Sampled Position Offset=&RandomRangeVector(Minimum=<0 0 -10>, Maximum=<0 0 10>)
    Skeletal Mesh=$System.Skeletal Mesh
   ) {enabled=false}
   &ParticleState()
   /SetVariables(Particles.ElectricityScale:NiagaraFloat=&FloatFromCurve(FloatCurve=DataInterface() {propertyValues {Curve {keys [
         {arriveTangent=4.1785, interpMode=RCIM_Cubic, leaveTangent=4.1785, tangentMode=RCTM_User, time=0, value=0}
         {arriveTangent=1.1179, interpMode=RCIM_Cubic, leaveTangent=1.1179, tangentMode=RCTM_User, time=0.2042, value=0.6228}
         {arriveTangent=0.0224, interpMode=RCIM_Cubic, leaveTangent=0.0224, tangentMode=RCTM_User, time=1, value=1}
        ]}}})) {isSetParameters=true}
   &ScaleMeshSize(Scale Factor=&MakeVector(X=$Particles.ElectricityScale, Y=$Particles.ElectricityScale, Z=1))
   /SetVariables(Particles.Position=&Lerp_Vector(
     A=Linked() {linkedVariable {name=Emitter.StartPos, type=NiagaraPosition}}
     B=Linked() {linkedVariable {name=Emitter.EndPos, type=NiagaraPosition}}
     Alpha=&ClampFloat(Float=&Add_Float(
       A=$Particles.Initial.InitialPosition
       B=&FloatFromCurve(
        FloatCurve=DataInterface() {propertyValues {Curve {keys [
            {arriveTangent=4.1785, interpMode=RCIM_Cubic, leaveTangent=4.1785, tangentMode=RCTM_User, time=0, value=0}
            {arriveTangent=1.1179, interpMode=RCIM_Cubic, leaveTangent=1.1179, tangentMode=RCTM_User, time=0.2042, value=0.6228}
            {arriveTangent=0.0224, interpMode=RCIM_Cubic, leaveTangent=0.0224, tangentMode=RCTM_User, time=1, value=1}
           ]}}}
        Scale Curve=0.2
       )
      ))
    )) {isSetParameters=true}
   &UpdateMeshOrientation(Rotation Rate=&RandomRangeFloat(Minimum=-0.5, Maximum=0.5), Rotation Vector=<0 0 1>)
   &ScaleColor(
    Scale Alpha=&Multiply_Float(
     A=&FloatFromCurve(
      FloatCurve=DataInterface() {propertyValues {Curve {keys [
          {interpMode=RCIM_Cubic, tangentMode=RCTM_User, time=0, value=0}
          {interpMode=RCIM_Cubic, tangentMode=RCTM_User, time=0.5, value=1}
          {tangentMode=RCTM_User, time=1, value=0}
         ]}}}
      Scale Curve=&FNTimeOfDayScalar(Night Scale=0.7, Morning Scale=0.3, Evening Scale=0.3)
     )
     B=&Multiply_Float(A=&RandomRangeFloat(Minimum=0, Maximum=1, Evaluation Type=::NewEnumerator1), B=1)
    )
    Scale RGB=&VectorFromFloat(Value=&Multiply_Float(
      A=&MakeCustomFloatFromBool(Bool=$User.bIsSwinging, True Float=20, False Float=10)
      B=&FloatFromCurve(
       FloatCurve=DataInterface() {propertyValues {Curve {keys [<at 0 1>,<at 0.2 0>,<at 0.3 1.25>,<at 0.4 0>,<at 0.7 0.75>,<at 1 0>]}}}
       Scale Curve=5
      )
     ))
    ScaleA=true
    ScaleRGB=true
   )
   &DynamicMaterialParameters(
    Index 0 Param 1=&Multiply_Float(
     A=1
     B=&FloatFromCurve(FloatCurve=DataInterface() {propertyValues {Curve {keys [
          {arriveTangent=-2.6097, interpMode=RCIM_Cubic, leaveTangent=-2.6097, tangentMode=RCTM_User, time=0, value=1.5}
          {arriveTangent=-0.6794, interpMode=RCIM_Cubic, leaveTangent=-0.6794, tangentMode=RCTM_User, time=1, value=0}
         ]}}})
    )
    Index 0 Param 2=&FloatFromCurve(
     FloatCurve=DataInterface() {propertyValues {Curve {keys [
         {arriveTangent=4.1785, interpMode=RCIM_Cubic, leaveTangent=4.1785, tangentMode=RCTM_User, time=0, value=0}
         {arriveTangent=1.1179, interpMode=RCIM_Cubic, leaveTangent=1.1179, tangentMode=RCTM_User, time=0.2042, value=0.6228}
         {arriveTangent=0.0224, interpMode=RCIM_Cubic, leaveTangent=0.0224, tangentMode=RCTM_User, time=1, value=1}
        ]}}}
     Scale Curve=0.1
    )
    Index 0 Param 3=1
    Index 0 Param 4=&FloatFromCurve(
     FloatCurve=DataInterface() {propertyValues {Curve {keys [{interpMode=RCIM_Cubic, tangentMode=RCTM_User, time=0, value=0}, {interpMode=RCIM_Cubic, tangentMode=RCTM_User, time=1, value=1}]}}}
     Scale Curve=0
    )
    Module.Param0WriteEnabled=true
    Module.Param1WriteEnabled=true
    Module.Param2WriteEnabled=true
    Module.Param3WriteEnabled=false
   )
   /SetVariables(Particles.Position=&Lerp_Vector(
     A=Linked() {linkedVariable {name=Emitter.StartPos, type=NiagaraPosition}}
     B=Linked() {linkedVariable {name=Emitter.EndPos, type=NiagaraPosition}}
     Alpha=&ClampFloat(Float=&Add_Float(
       A=$Particles.Initial.InitialPosition
       B=&FloatFromCurve(
        FloatCurve=DataInterface() {propertyValues {Curve {keys [
            {arriveTangent=4.1785, interpMode=RCIM_Cubic, leaveTangent=4.1785, tangentMode=RCTM_User, time=0, value=0}
            {arriveTangent=1.1179, interpMode=RCIM_Cubic, leaveTangent=1.1179, tangentMode=RCTM_User, time=0.2042, value=0.6228}
            {arriveTangent=0.0224, interpMode=RCIM_Cubic, leaveTangent=0.0224, tangentMode=RCTM_User, time=1, value=1}
           ]}}}
        Scale Curve=0.2
       )
      ))
    )) {isSetParameters=true}
   /SetVariables(Particles.Glow Dy02:Vector4f=&MakeVector4(
     X=&Multiply_Float(
      A=3
      B=&FloatFromCurve(FloatCurve=DataInterface() {propertyValues {Curve {keys [
           {arriveTangent=-2.6097, interpMode=RCIM_Cubic, leaveTangent=-2.6097, tangentMode=RCTM_User, time=0, value=1.5}
           {arriveTangent=-0.6794, interpMode=RCIM_Cubic, leaveTangent=-0.6794, tangentMode=RCTM_User, time=1, value=0}
          ]}}})
     )
     Y=0
     Z=0
     W=0
    )) {isSetParameters=true}
   /SetVariables(Particles.Glow color:LinearColor=&MakeLinearColorFromVectorAndFloat(
     "Vector (RGB)"=&Multiply_VectorByFloat(
      Vector=&MakeVectorFromLinearColorRGB(Color=$User.VariantColor)
      Float=&MakeCustomFloatFromBool(Bool=$User.bIsSwinging, True Float=2.5, False Float=1)
     )
     "Float (Alpha)"=&Multiply_Float(
      A=&Multiply_Float(
       A=&FloatFromCurve(
        FloatCurve=DataInterface() {propertyValues {Curve {keys [
            {interpMode=RCIM_Cubic, tangentMode=RCTM_User, time=0, value=0}
            {arriveTangent=-2.4253, interpMode=RCIM_Cubic, leaveTangent=-2.4253, tangentMode=RCTM_User, time=0.4, value=0.975}
            {tangentMode=RCTM_User, time=1, value=0}
           ]}}}
        Scale Curve=&FNTimeOfDayScalar(Night Scale=0.7, Morning Scale=0.3, Evening Scale=0.3)
       )
       B=&Multiply_Float(A=&RandomRangeFloat(Minimum=0, Maximum=1, Evaluation Type=::NewEnumerator1), B=1)
      )
      B=0.5
     )
    )) {isSetParameters=true}
   &FN_Sprite_Attributes(
    Color=&MultiplyLinearColor(
     Linear Color=$User.VariantColor
     Scale Factor=&MakeFloat4FromVectorAndFloat(
      "Vector (RGB)"=<1 1 1>
      "Float (Alpha)"=&FloatFromCurve(FloatCurve=DataInterface() {propertyValues {Curve {keys [<at 0 1>,<at 1 0>]}}}, Scale Curve=0.05)
     )
    )
    Create Camera Offset Attribute=true
    Create Color Attribute=true
    Create SpriteSize Attribute=true
    Create Visibility Attribute=true
    SpriteSize=&Vector2DFromFloat(Value=78.7742)
    Visibility Tag:NiagaraInt32=&MakeIntFromFloat(Float=&RandomRangeFloat(Minimum=0, Maximum=1), Conversion Mode=::NewEnumerator1)
    Write to Module Namespace=true
   )
  ]
  rendererClasses [&NiagaraMeshRendererProperties, &NiagaraSpriteRendererProperties]
  renderers [
   {rendererClass=&NiagaraMeshRendererProperties, rendererIndex=0}
   {rendererClass=&NiagaraMeshRendererProperties, rendererIndex=1}
   {rendererClass=&NiagaraSpriteRendererProperties, rendererIndex=2}
  ]
 }
 Lightning001 {
  bEnabled=false
  emitterSpawnScript [/SetVariables(Emitter.EndPos:NiagaraPosition=<-1 1 75>, Emitter.StartPos:NiagaraPosition=<-3.5 1 20>) {isSetParameters=true}]
  emitterUpdateScript [
   &EmitterState(
    Inactive Response=::NewEnumerator1
    Life Cycle Mode=::NewEnumerator1
    Loop Duration=&RandomRangeFloat(Minimum=2, Maximum=7)
    Recalculate Duration Each Loop=true
    UseLoopDelay=true
   ) {module=EmitterState001}
   &SpawnRate(SpawnRate=5, Use Spawn Probability=false) {enabled=false}
   &SpawnRate(
    SpawnRate=&Multiply_Float(
     A=&Multiply_Float(
      A=&RandomRangeFloat(Minimum=6, Maximum=8)
      B=&Waveform(
       Waveform Count=::NewEnumerator1
       "Waveform [1]"=::NewEnumerator0
       "[1] Frequency"=1
       "[1] Amplitude Min/Max"=<0 1>
       "[1] Scale"=1.5
       "Waveform [2] Blend"=::NewEnumerator2
       "Waveform [2]"=::NewEnumerator12
       "[2] Input Position"=0
       "[2] Phase"=$Emitter.Age
       "[2] Phase Scale"=1
       "[2] Compound Frequency"=2
       "[2] Frequency"=12
       "[2] Amplitude Min/Max"=<0.1 1>
       "[2] Scale"=2
       "[2] Bias"=0
       "[2] Exponent"=1
       "[2] Pulse Duration"=0.1
       "[2] Clip Triangle Min/Max"=<0 0>
       "[2] Random Seed"=0
      )
     )
     B=&FloatFromCurve(FloatCurve=DataInterface() {propertyValues {Curve {keys [<at 0 1.5>,<at 0.2 1.125>,<at 1 1>]}}}, CurveIndex=$Emitter.NormalizedLoopAge)
    )
    Use Spawn Probability=false
   ) {module=SpawnRate001}
  ]
  particleSpawnScript [
   &InitializeParticle(
    Color=<rgba 0.3329 1.946 3 1>
    Lifetime=&RandomRangeFloat(Minimum=0.2, Maximum=0.3)
    Mesh Scale Mode=::NewEnumerator3
    UsePositionOffset=false
    Write Material Random=true
   )
   /SetVariables(
    Particles.InvertLightning:NiagaraBool=&RandomBool()
    Particles.LightningIndex=1
    Particles.MirrorLightning:NiagaraBool=&RandomBool()
   ) {isSetParameters=true}
   /SetVariables(Particles.LightningIndex:NiagaraInt32=&Add_Integer(
     A=0
     B=&ModuloInt(
      A=&SelectIntFromArray(
       Direct Array Index=&ReturnParticleIDAttributes()
       Direct Array Mode=::NewEnumerator1
       Int Selection Array=$User.LightningOrder
      )
      B=5
     )
    )) {isSetParameters=true}
   /SetVariables(Particles.Scale=&Multiply_Vector(
     A=&MakeCustomVectorFromBool(Bool=$Particles.MirrorLightning, True Vector=<1 1 -1>, False Vector=<1 1 1>)
     B=&RandomRangeVector(Minimum=<0.2 0.2 0.6>, Maximum=<0.5 0.5 0.8>)
    )) {isSetParameters=true}
   /SetVariables(Particles.InitialPosition:NiagaraFloat=&RandomRangeFloat(Minimum=0.1, Maximum=0.6)) {isSetParameters=true}
   /SetVariables(Particles.Position=&Lerp_Vector(A=$Emitter.StartPos, B=$Emitter.EndPos, Alpha=$Particles.InitialPosition)) {isSetParameters=true}
   &InitialMeshOrientation(
    Mesh Axis To Orient=<0 0 1>
    Mesh Orientation Mode=::NewEnumerator2
    Orientation Vector=&NormalizeVector(
     Vector To Normalize=&Subtract_Position(A=$Emitter.StartPos, B=$Emitter.EndPos)
     Safe Normalize=true
     Safe Fallback Threshold=0
     Safe Fallback Vector=<0 0 0.001>
    )
    Rotation=&MakeVector(X=0, Y=0, Z=&RandomRangeFloat(Minimum=0, Maximum=1))
    Use Rotation Vector=true
   )
  ]
  particleUpdateScript [
   &SocketLocation(
    SamplePosOffset=false
    Socket Index=$Particles.SocketIndex
    Socket Reader=$System.Socket Reader Interface
    Socket Sampling Mode=::NewEnumerator3
    VelocitySpeedLimit=false
   ) {enabled=false}
   &DynamicMaterialParameters(
    Index 0 Param 1=&RandomRangeFloat(Minimum=2, Maximum=5)
    Index 0 Param 2=&Divide_Float(A=&RandomRangeFloat(Minimum=1, Maximum=1.5), B=2)
    Index 0 Param 3=&FloatFromCurve(FloatCurve=DataInterface() {propertyValues {Curve {keys [<at 0 1>,<at 0.1 0.375>,<at 0.5 0.05>,<at 1 0>]}}})
    Index 0 Param 4=&MakeCustomFloatFromBool(Bool=$Particles.InvertLightning, True Float=1, False Float=0)
    Index 1 Param 1=&MakeFloatFromVector(Vector=$Particles.Scale, Channel=::NewEnumerator0)
    Index 1 Param 2=&MakeFloatFromVector(Vector=$Particles.Scale, Channel=::NewEnumerator1)
    Index 1 Param 3=&MakeFloatFromVector(Vector=$Particles.Scale, Channel=::NewEnumerator2)
    Module.Index0Param0Write=true
    Module.Index0Param1Write=true
    Module.Index0Param2Write=true
    Module.Index0Param3Write=true
    Module.Index1Param0Write=true
    Module.Index1Param1Write=true
    Module.Index1Param2Write=true
    Module.Index1Param3Write=true
    Write Parameter Index 1=true
   )
   &ParticleState()
   /SetVariables(Particles.Position=&Lerp_Vector(
     A=$Emitter.StartPos
     B=$Emitter.EndPos
     Alpha=&ClampFloat(Float=&Add_Float(
       A=$Particles.Initial.InitialPosition
       B=&FloatFromCurve(
        FloatCurve=DataInterface() {propertyValues {Curve {keys [
            {arriveTangent=4.1785, interpMode=RCIM_Cubic, leaveTangent=4.1785, tangentMode=RCTM_User, time=0, value=0}
            {arriveTangent=1.1179, interpMode=RCIM_Cubic, leaveTangent=1.1179, tangentMode=RCTM_User, time=0.2042, value=0.6228}
            {arriveTangent=0.0224, interpMode=RCIM_Cubic, leaveTangent=0.0224, tangentMode=RCTM_User, time=1, value=1}
           ]}}}
        Scale Curve=0.3
       )
      ))
    )) {isSetParameters=true}
   &Color(Color=&ColorFromCurve(ColorCurve=DataInterface() {propertyValues {
       AlphaCurve {keys [<at 0 0>, {arriveTangent=0.3873, interpMode=RCIM_Cubic, leaveTangent=0.3873, time=0.0998, value=1}, <at 0.7746 0.3>, <at 1 0 RCIM_Cubic>]}
       BlueCurve {keys [<at 0 66 RCIM_Cubic>,<at 0.988 18.5625>]}
       GreenCurve {keys [<at 0 32.9999 RCIM_Cubic>,<at 0.988 9.2812>]}
       RedCurve {keys [<at 0 0 RCIM_Cubic>,<at 0.988 0>]}
      }})) {enabled=false}
   &ScaleColor(
    Linear Color Curve=DataInterface() {propertyValues {
      AlphaCurve {keys [<at 0 0>, {arriveTangent=0.3873, interpMode=RCIM_Cubic, leaveTangent=0.3873, time=0.0998, value=1}, <at 0.7746 0.3>, <at 1 0 RCIM_Cubic>]}
      BlueCurve {keys [<at 0.0056 10>,<at 0.9952 1>]}
      GreenCurve {keys [<at 0.0056 10>,<at 0.9952 1>]}
      RedCurve {keys [<at 0.0056 10>,<at 0.9952 1>]}
     }}
    Scale Mode=::NewEnumerator2
   )
   /SetVariables(Particles.ElectricityScale:NiagaraFloat=&FloatFromCurve(FloatCurve=DataInterface() {propertyValues {Curve {keys [
         {arriveTangent=4.1785, interpMode=RCIM_Cubic, leaveTangent=4.1785, tangentMode=RCTM_User, time=0, value=0}
         {arriveTangent=1.1179, interpMode=RCIM_Cubic, leaveTangent=1.1179, tangentMode=RCTM_User, time=0.2042, value=0.6228}
         {arriveTangent=0.0224, interpMode=RCIM_Cubic, leaveTangent=0.0224, tangentMode=RCTM_User, time=1, value=1}
        ]}}})) {isSetParameters=true}
   &ScaleMeshSize(Scale Factor=&MakeVector(
     X=$Particles.ElectricityScale
     Y=$Particles.ElectricityScale
     Z=&FloatFromCurve(FloatCurve=DataInterface() {propertyValues {Curve {keys [<at 0 0.7>,<at 1 1>]}}})
    ))
   &UpdateMeshOrientation(Rotation Rate=&RandomRangeFloat(Minimum=-1, Maximum=1), Rotation Vector=<0 0 1>)
   /SetVariables(Particles.Glow color:LinearColor=&MakeLinearColorFromVectorAndFloat(
     "Vector (RGB)"=&Multiply_VectorByFloat(
      Vector=&MakeVectorFromLinearColorRGB(Color=<rgba 0.2221 0.767 1 1>)
      Float=&MakeCustomFloatFromBool(Bool=$User.bIsSwinging, True Float=2.5, False Float=1)
     )
     "Float (Alpha)"=&Multiply_Float(
      A=&Multiply_Float(
       A=&FloatFromCurve(
        FloatCurve=DataInterface() {propertyValues {Curve {keys [
            {interpMode=RCIM_Cubic, tangentMode=RCTM_User, time=0, value=0}
            {interpMode=RCIM_Cubic, tangentMode=RCTM_User, time=0.5, value=1}
            {tangentMode=RCTM_User, time=1, value=0}
           ]}}}
        Scale Curve=&FNTimeOfDayScalar(Night Scale=0.7, Morning Scale=0.3, Evening Scale=0.3)
       )
       B=&Multiply_Float(A=&RandomRangeFloat(Minimum=0, Maximum=1, Evaluation Type=::NewEnumerator1), B=1)
      )
      B=0.5
     )
    )) {isSetParameters=true}
   &FN_Sprite_Attributes(
    Color=&MultiplyLinearColor(Linear Color=$Particles.Glow color, Scale Factor=<1 1 1 0.05>)
    Create Camera Offset Attribute=true
    Create Color Attribute=true
    Create SpriteSize Attribute=true
    Create Visibility Attribute=true
    SpriteSize=&Vector2DFromFloat(Value=78.7742)
    Visibility Tag:NiagaraInt32=&MakeIntFromFloat(Float=&RandomRangeFloat(Minimum=0, Maximum=1), Conversion Mode=::NewEnumerator1)
    Write to Module Namespace=true
   )
  ]
  rendererClasses [&NiagaraMeshRendererProperties, &NiagaraSpriteRendererProperties]
  renderers [
   {rendererClass=&NiagaraMeshRendererProperties, rendererIndex=0}
   {rendererClass=&NiagaraMeshRendererProperties, rendererIndex=1}
   {rendererClass=&NiagaraSpriteRendererProperties, rendererIndex=2}
  ]
 }
}
summary {
 systemName=NS_Pickaxe_RankedTrooper_Idle
 emitters [
  {emitterName=Glows, rendererClasses [&NiagaraMeshRendererProperties, &NiagaraSpriteRendererProperties]}
  {emitterName=Lightning, rendererClasses [&NiagaraMeshRendererProperties, &NiagaraSpriteRendererProperties]}
  {
   bEnabled=false
   emitterName=Lightning001
   rendererClasses [&NiagaraMeshRendererProperties, &NiagaraSpriteRendererProperties]
  }
 ]
 userVariables [
  User.bIsFrontEndPreview {type=NiagaraBool, defaultValue=true}
  User.bIsSwinging {type=NiagaraBool}
  User.FXScale {type=NiagaraFloat, defaultValue=1.01, description=""}
  User.VariantColor {type=LinearColor, description="", defaultValue {struct=LinearColor, value=<rgba 0.113 0.7545 1 1>}}
 ]
}
