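# Niagara swing trail effect for the RankedTrooper pickaxe, drawing a blade trail with electric accents.
# Three ribbon emitters use anim-trail modules on bone or socket indices: BladeTrailMain, Lightning001 between two points, and LightningEdge at a third.
# Each has two ribbon renderers, with colors and fades from system base color, swing, electricity and time-of-day values.
# LightningEdge flickers via waveforms, and both lightning layers animate dynamic material parameters.
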NiagaraSystem /BRCosmetics/Weapons/FORT_Melee/Pickaxe_RankedTrooper/FX/NS_Pickaxe_RankedTrooper_Trails
dependencies {
 usedDataInterfaces [
  &NiagaraDataInterfaceCurve
  &NiagaraDataInterfaceColorCurve
  &NiagaraDataInterfaceSocketReader
  &NiagaraDataInterfaceSkeletalMesh
 ]
 usedDynamicInputs []
 usedModules [
  &self:SystemSpawnScript.SystemScriptSource.SystemScriptGraph.NiagaraNodeAssignment_4.SetVariables_#1
  &self:SystemSpawnScript.SystemScriptSource.SystemScriptGraph.NiagaraNodeAssignment_5.SetVariables_#2
  &self:SystemSpawnScript.SystemScriptSource.SystemScriptGraph.NiagaraNodeAssignment_0.SetVariables_#3
  &SystemState
  &self:SystemSpawnScript.SystemScriptSource.SystemScriptGraph.NiagaraNodeAssignment_1.SetVariables_#4
  &self:SystemSpawnScript.SystemScriptSource.SystemScriptGraph.NiagaraNodeAssignment_2.SetVariables_#5
  &self:SystemSpawnScript.SystemScriptSource.SystemScriptGraph.NiagaraNodeAssignment_3.SetVariables_#6
 ]
}
emitters {
 BladeTrailMain {
  emitterUpdateScript [
   &EmitterState()
   &FN_SpawnAnimTrail()
  ]
  particleSpawnScript [
   &FN_InitializeAnimTrailBoneIndecies(First Bone or Socket Index=1, Second Bone or Socket Index=0)
   &FN_InitializeAnimTrail(Lifetime=0.2)
   /SetVariables(Particles.RibbonWidthDark:NiagaraFloat=&Multiply_Float(A=$Particles.RibbonWidth, B=1)) {isSetParameters=true}
   &AddVelocity(Velocity=<0 1000 0>) {enabled=false}
  ]
  particleUpdateScript [
   &SolveForcesAndVelocity(Clamp Velocity=false, Limit Acceleration=false) {enabled=false}
   &FN_UpdateAnimTrail() {module=FN_UpdateAnimTrail001}
   &ScaleColor(
    Scale Alpha=&FloatFromCurve(
     FloatCurve=DataInterface() {propertyValues {Curve {keys [{arriveTangent=-0.0018, interpMode=RCIM_Cubic, leaveTangent=-0.2406, tangentMode=RCTM_Break, time=0.4, value=1}, <at 1 0 RCIM_Cubic>]}}}
     Scale Curve=&Multiply_Float(A=&Multiply_Float(A=$System.SwingGeneralFade, B=$System.TODFloat), B=0.7)
    )
    Scale RGB=&MakeVectorFromLinearColorRGB(Color=&Multiply_LinearColorByFloat(
      Linear Color=$System.BaseColor
      Float=&FloatFromCurve(
       FloatCurve=DataInterface() {propertyValues {Curve {keys [{arriveTangent=-6.5622, interpMode=RCIM_Cubic, leaveTangent=-6.5622, tangentMode=RCTM_User, time=0, value=2}, <at 0.5 0.5 RCIM_Cubic>]}}}
       CurveIndex=&Multiply_Float(A=$Particles.NormalizedAge, B=6)
      )
     ))
    ScaleA=true
    ScaleRGB=true
   )
   /SetVariables(Particles.ColorGlow:LinearColor=&MultiplyLinearColor(
     Linear Color=&HueShiftLinearColor(
      Hue=&FloatFromCurve(
       FloatCurve=DataInterface() {propertyValues {Curve {keys [{tangentMode=RCTM_User, time=0, value=0}, {tangentMode=RCTM_User, time=1, value=1}]}}}
       Scale Curve=-0.1
      )
      Saturation=1
      Value=&FloatFromCurve(FloatCurve=DataInterface() {propertyValues {Curve {keys [<at 0 1>,<at 1 0>]}}})
      Linear Color=$System.BaseColor
     )
     Scale Factor=&MakeFloat4FromVectorAndFloat(
      "Vector (RGB)"=<1 1 1>
      "Float (Alpha)"=&FloatFromCurve(
       FloatCurve=DataInterface() {propertyValues {Curve {keys [{arriveTangent=-0.0018, interpMode=RCIM_Cubic, leaveTangent=-0.2406, tangentMode=RCTM_Break, time=0.4, value=1}, <at 1 0 RCIM_Cubic>]}}}
       Scale Curve=&Multiply_Float(A=&Multiply_Float(A=$System.SwingGeneralFade, B=$System.TODFloat), B=0.7)
      )
     )
    )) {isSetParameters=true}
  ]
  rendererClasses [&NiagaraRibbonRendererProperties]
  renderers [
   {rendererClass=&NiagaraRibbonRendererProperties}
   {rendererClass=&NiagaraRibbonRendererProperties, rendererIndex=1}
  ]
 }
 Lightning001 {
  emitterUpdateScript [
   &EmitterState()
   &FN_SpawnAnimTrail()
   /SetVariables(Emitter.rand:NiagaraFloat=&RandomRangeFloat(Minimum=1, Maximum=1, Evaluation Type=::NewEnumerator1)) {isSetParameters=true}
  ]
  particleSpawnScript [
   &FN_InitializeAnimTrailBoneIndecies(First Bone or Socket Index=1, Second Bone or Socket Index=0)
   &FN_InitializeAnimTrail(
    AnimTrailWidth=1.25
    Color=&HueShiftLinearColor(Hue=0, Saturation=1, Value=3, Linear Color=$System.BaseColor)
    Lifetime=0.27
   )
   /SetVariables(Particles.Position=&Lerp_Vector(A=$System.SampleSkeletalMesh.SampledPosition, B=$System.SampleSkeletalMesh001.SampledPosition, Alpha=0.4)) {isSetParameters=true}
  ]
  particleUpdateScript [
   &FN_UpdateAnimTrail()
   &ParticleState()
   &ScaleColor(
    Scale Alpha=&FloatFromCurve(
     FloatCurve=DataInterface() {propertyValues {Curve {keys [{interpMode=RCIM_Cubic, tangentMode=RCTM_User, time=0, value=1}, {interpMode=RCIM_Cubic, tangentMode=RCTM_User, time=1, value=0}]}}}
     Scale Curve=$Emitter.rand
    )
    Scale RGB=&VectorFromFloat(Value=&FloatFromCurve(
      FloatCurve=DataInterface() {propertyValues {Curve {keys [
          <at 0 0.65 RCIM_Cubic>
          {arriveTangent=0.25, interpMode=RCIM_Cubic, leaveTangent=0.25, time=0.1, value=3}
          {arriveTangent=-3.3286, interpMode=RCIM_Cubic, leaveTangent=-3.3286, time=0.4, value=0.75}
          {arriveTangent=-1.25, interpMode=RCIM_Cubic, leaveTangent=-1.25, time=0.7009, value=1}
          {interpMode=RCIM_Cubic, leaveTangent=-8.0089, time=1, value=0}
         ]}}}
      Scale Curve=2
     ))
    ScaleA=true
    ScaleRGB=true
   )
   &DynamicMaterialParameters(
    Index 0 Param 2=&FloatFromCurve(FloatCurve=DataInterface() {propertyValues {Curve {keys [
         {arriveTangent=0.2955, interpMode=RCIM_Cubic, leaveTangent=0.2955, tangentMode=RCTM_User, time=0, value=0}
         {arriveTangent=-2.2131, interpMode=RCIM_Cubic, leaveTangent=-2.2131, tangentMode=RCTM_User, time=1, value=-1}
        ]}}})
    Module.Param0WriteEnabled=false
    Module.Param1WriteEnabled=true
    Module.Param2WriteEnabled=false
    Module.Param3WriteEnabled=false
   )
   /SetVariables(Particles.Trail Color02:LinearColor=&MakeLinearColorFromVectorAndFloat(
     "Vector (RGB)"=&MakeVectorFromLinearColorRGB(Color=$System.BaseColor)
     "Float (Alpha)"=&Multiply_Float(
      A=$System.SwingGeneralFade
      B=&FloatFromCurve(
       FloatCurve=DataInterface() {propertyValues {Curve {keys [
           {interpMode=RCIM_Cubic, tangentMode=RCTM_User, time=0, value=1}
           {arriveTangent=-1.4244, interpMode=RCIM_Cubic, leaveTangent=-1.4244, tangentMode=RCTM_User, time=1, value=0}
          ]}}}
       Scale Curve=$Emitter.rand
      )
     )
    )) {isSetParameters=true}
   /SetVariables(Particles.DY2:Vector4f=&MakeVector4(X=1, Y=1, Z=1, W=1)) {isSetParameters=true}
  ]
  rendererClasses [&NiagaraRibbonRendererProperties]
  renderers [
   {rendererClass=&NiagaraRibbonRendererProperties}
   {rendererClass=&NiagaraRibbonRendererProperties, rendererIndex=1}
  ]
 }
 LightningEdge {
  emitterUpdateScript [
   &EmitterState()
   &FN_SpawnAnimTrail()
  ]
  particleSpawnScript [
   &FN_InitializeAnimTrailBoneIndecies(First Bone or Socket Index=2, Second Bone or Socket Index=2)
   &FN_InitializeAnimTrail(
    AnimTrailWidth=&MaskFloatBySpawnGroup(Masked Float=8, Passthrough Float=7)
    Color=$System.BaseColor
    Lifetime=&MaskFloatBySpawnGroup(Masked Float=0.082, Passthrough Float=0.05)
   )
  ]
  particleUpdateScript [
   &FN_UpdateAnimTrail()
   /SetVariables(Particles.Waveform:NiagaraFloat=&Waveform("Waveform [1]"=::NewEnumerator4, "[1] Phase Scale"=4, "[1] Frequency"=3.5, "[1] Amplitude Min/Max"=<0.1 1>)) {isSetParameters=true}
   &ScaleColor(
    Scale Alpha=&FloatFromCurve(
     FloatCurve=DataInterface() {propertyValues {Curve {keys [<at 0.3727 1 RCIM_Cubic>,<at 1 0 RCIM_Constant>]}}}
     Scale Curve=$System.ElectricityFade
    )
    Scale RGB=&MakeVectorFromLinearColorRGB(Color=&MultiplyLinearColor(
      Linear Color=&ColorFromCurve(ColorCurve=DataInterface() {propertyValues {
         AlphaCurve {keys []}
         BlueCurve {keys [<at 0.3243 1 RCIM_Cubic>,<at 0.9154 0.875 RCIM_Cubic>]}
         GreenCurve {keys [<at 0.3243 1 RCIM_Cubic>,<at 0.9154 0.7827 RCIM_Cubic>]}
         RedCurve {keys [<at 0.3243 1 RCIM_Cubic>,<at 0.9154 0.229 RCIM_Cubic>]}
        }})
      Scale Factor=&MakeVector4_FromFloat(Float=&FNTimeOfDayScalar(Night Scale=0.6, Morning Scale=0.8, Evening Scale=0.8))
     ))
    ScaleA=true
    ScaleRGB=true
   )
   &DynamicMaterialParameters(
    Index 0 Param 1=1
    Index 0 Param 2=&FloatFromCurve(FloatCurve=DataInterface() {propertyValues {Curve {keys [
         {arriveTangent=-0.377, interpMode=RCIM_Cubic, leaveTangent=-0.377, tangentMode=RCTM_User, time=0.4149, value=-0.6131}
         {arriveTangent=-1.4738, interpMode=RCIM_Cubic, leaveTangent=-1.4738, tangentMode=RCTM_User, time=1, value=-1}
        ]}}})
    Index 1 Param 1=&MaskFloatBySpawnGroup(Masked Float=0, Passthrough Float=0.4)
    Module.Index0Float0Write=true
    Module.Index0Float1Write=true
    Module.Index0Float2Write=true
    Module.Index0Float3Write=true
    Module.Index1Float0Write=true
    Module.Index1Float1Write=true
    Module.Index1Float2Write=true
    Module.Index1Float3Write=true
    Write Parameter Index 1=::NewEnumerator1
   )
   &ScaleColor(
    Scale Alpha=$Particles.Waveform
    Scale RGB=&VectorFromFloat(Value=&Waveform("Waveform [1]"=::NewEnumerator1, "[1] Phase Scale"=2, "[1] Frequency"=7, "[1] Amplitude Min/Max"=<0.9 1.2>))
    ScaleA=true
    ScaleRGB=true
   ) {module=ScaleColor001}
   &ScaleRibbonWidth(Ribbon Width Scale=&FloatFromCurve(FloatCurve=DataInterface() {propertyValues {Curve {keys [
         <at 0 0.6636 RCIM_Cubic>
         {arriveTangent=1.6411, interpMode=RCIM_Cubic, leaveTangent=1.6411, tangentMode=RCTM_User, time=0.0823, value=1}
         {arriveTangent=-0.4875, interpMode=RCIM_Cubic, leaveTangent=-0.4875, time=0.3314, value=1}
         <at 1 0.5526>
        ]}}}))
   /SetVariables(Particles.GlowColor=&MakeLinearColorFromVectorAndFloat(
     "Vector (RGB)"=&MakeVectorFromLinearColorRGB(Color=&MultiplyLinearColor(
       Linear Color=&ColorFromCurve(ColorCurve=DataInterface() {propertyValues {
          AlphaCurve {keys []}
          BlueCurve {keys [<at 0.3243 5 RCIM_Cubic>,<at 0.9154 15 RCIM_Cubic>]}
          GreenCurve {keys [<at 0.3243 4.1806 RCIM_Cubic>,<at 0.9154 11.5994 RCIM_Cubic>]}
          RedCurve {keys [<at 0.3243 1.3983 RCIM_Cubic>,<at 0.9154 1.8893 RCIM_Cubic>]}
         }})
       Scale Factor=&MakeVector4_FromFloat(Float=&FNTimeOfDayScalar(Night Scale=0.6, Morning Scale=0.8, Evening Scale=0.8))
      ))
     "Float (Alpha)"=&FloatFromCurve(
      FloatCurve=DataInterface() {propertyValues {Curve {keys [<at 0.3727 9 RCIM_Cubic>, {interpMode=RCIM_Cubic, tangentMode=RCTM_User, time=1, value=0}]}}}
      Scale Curve=&Multiply_Float(A=$System.ElectricityFade, B=$Particles.Waveform)
     )
    )) {isSetParameters=true}
   /SetVariables(Particles.GlowWidth:NiagaraFloat=&FloatFromCurve(
     FloatCurve=DataInterface() {propertyValues {Curve {keys [
         <at 0 0.6636 RCIM_Cubic>
         {arriveTangent=1.6411, interpMode=RCIM_Cubic, leaveTangent=1.6411, tangentMode=RCTM_User, time=0.0823, value=1}
         {arriveTangent=-0.4875, interpMode=RCIM_Cubic, leaveTangent=-0.4875, time=0.3314, value=1}
         <at 1 0.5526>
        ]}}}
     Scale Curve=12
    )) {isSetParameters=true}
  ]
  rendererClasses [&NiagaraRibbonRendererProperties]
  renderers [
   {rendererClass=&NiagaraRibbonRendererProperties}
   {rendererClass=&NiagaraRibbonRendererProperties, rendererIndex=1}
  ]
 }
}
summary {
 systemName=NS_Pickaxe_RankedTrooper_Trails
 emitters [
  {emitterName=BladeTrailMain, rendererClasses [&NiagaraRibbonRendererProperties]}
  {emitterName=LightningEdge, rendererClasses [&NiagaraRibbonRendererProperties]}
  {emitterName=Lightning001, rendererClasses [&NiagaraRibbonRendererProperties]}
 ]
 userVariables [
  User.bisSwinging {type=NiagaraBool}
  User.RibbonIDOffset {type=NiagaraInt32}
  User.VariantColor {type=LinearColor, description="", defaultValue {struct=LinearColor, value=<rgba 0.113 0.7545 1 1>}}
 ]
}
